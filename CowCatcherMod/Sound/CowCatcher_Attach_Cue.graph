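# Sound cue for attaching the cow catcher vehicle mod, layering three randomized elements through a mixer.
# An attach sound plays alongside a delayed, quieter suspension bounce borrowed from off-road tire audio and a tool sound.
# Random nodes pick variants, with pitch, volume and timing variation on selected branches.
# Custom attenuation adds logarithmic falloff with distance low-pass filtering; concurrency is limited per owner under the vehicles sound class.
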SoundCue /CowCatcherMod/Sound/CowCatcher_Attach_Cue
root=Mixer_0
nodes [
 {id=Delay_0, children [Modulator_1], properties {delayMax=0.3, delayMin=0.3}}
 {id=Delay_1, children [WavePlayer_13], properties {delayMax=0.15, delayMin=0.05}}
 {id=Mixer_0, children [Modulator_0,Delay_0,Modulator_2], properties {inputVolume [1,0.4,0.8]}}
 {id=Modulator_0, children [Random_0], properties {pitchMax=1.1, pitchMin=1, volumeMax=0.85, volumeMin=0.75}}
 {id=Modulator_1, children [Random_1], properties {pitchMax=1.15, pitchMin=0.85}}
 {id=Modulator_2, children [Random_2], properties {pitchMax=1.15, pitchMin=0.85}}
 {id=Modulator_3, children [WavePlayer_0], properties {pitchMax=0.9, pitchMin=0.8, volumeMax=0.95, volumeMin=0.85}}
 {id=Modulator_4, children [WavePlayer_1], properties {volumeMax=1.3, volumeMin=1.2}}
 {id=Random_0, children [WavePlayer_2,WavePlayer_3,WavePlayer_4,WavePlayer_5], properties {weights [1,1,1,1]}}
 {
  id=Random_1
  children [WavePlayer_6, WavePlayer_7, WavePlayer_8, WavePlayer_9, WavePlayer_10, WavePlayer_11]
  properties {weights [1,1,1,1,1,1]}
 }
 {id=Random_2, children [Modulator_3, Modulator_4, WavePlayer_12, Delay_1], properties {weights [1,1,1,1]}}
 {id=WavePlayer_0, properties {soundWaveAssetPtr=&/CowCatcher_AttachTool_01}}
 {id=WavePlayer_1, properties {soundWaveAssetPtr=&/CowCatcher_AttachTool_02}}
 {id=WavePlayer_10, properties {soundWaveAssetPtr=/ValetMods/Audio/OffRoadTire/Veh_Valet_Susp_ModBounce_Med_05}}
 {id=WavePlayer_11, properties {soundWaveAssetPtr=/ValetMods/Audio/OffRoadTire/Veh_Valet_Susp_ModBounce_Med_06}}
 {id=WavePlayer_12, properties {soundWaveAssetPtr=&/CowCatcher_AttachTool_03}}
 {id=WavePlayer_13, properties {soundWaveAssetPtr=&/CowCatcher_AttachTool_04}}
 {id=WavePlayer_2, properties {soundWaveAssetPtr=&/CowCatcher_Attach_01}}
 {id=WavePlayer_3, properties {soundWaveAssetPtr=&/CowCatcher_Attach_02}}
 {id=WavePlayer_4, properties {soundWaveAssetPtr=&/CowCatcher_Attach_03}}
 {id=WavePlayer_5, properties {soundWaveAssetPtr=&/CowCatcher_Attach_04}}
 {id=WavePlayer_6, properties {soundWaveAssetPtr=/ValetMods/Audio/OffRoadTire/Veh_Valet_Susp_ModBounce_Med_01}}
 {id=WavePlayer_7, properties {soundWaveAssetPtr=/ValetMods/Audio/OffRoadTire/Veh_Valet_Susp_ModBounce_Med_02}}
 {id=WavePlayer_8, properties {soundWaveAssetPtr=/ValetMods/Audio/OffRoadTire/Veh_Valet_Susp_ModBounce_Med_03}}
 {id=WavePlayer_9, properties {soundWaveAssetPtr=/ValetMods/Audio/OffRoadTire/Veh_Valet_Susp_ModBounce_Med_04}}
]
properties {
 bOverrideAttenuation=true
 bOverrideConcurrency=true
 maxDistance=5600
 soundClassObject=&Fort_Vehicles
 volumeMultiplier=0.8
 attenuationOverrides {
  bAttenuateWithLPF=true
  distanceAlgorithm=Logarithmic
  falloffDistance=5000
  lPFFrequencyAtMax=1000
  lPFRadiusMax=4000
  lPFRadiusMin=1000
  nonSpatializedRadiusStart=600
  stereoSpread=1000
  attenuationShapeExtents {x=600}
 }
 concurrencyOverrides {bLimitToOwner=true, maxCount=5, retriggerTime=0.3}
}
