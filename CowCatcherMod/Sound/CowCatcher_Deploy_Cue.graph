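# Sound cue for the CowCatcher mod's deploy sound, routed to the Fort_Vehicles sound class.
# A mixer blends three modulated layers: a random pick of four main waves, a random pick of three LayerA waves and a single LayerG wave.
# An outer modulator wraps the mix for further variation.
# Spatial falloff references the CowCatcher_Deploy_Att asset, and a concurrency override sets a short retrigger time.
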SoundCue /CowCatcherMod/Sound/CowCatcher_Deploy_Cue
root=Modulator_5
nodes [
 {id=Mixer_1, children [Modulator_0,Modulator_3,Modulator_4], properties {inputVolume [0.9,1.3,0.8]}}
 {id=Modulator_0, children [Random_1], properties {pitchMax=1.15, volumeMax=0.85, volumeMin=0.75}}
 {id=Modulator_3, children [Random_3], properties {pitchMax=1.1, pitchMin=1}}
 {id=Modulator_4, children [WavePlayer_17], properties {pitchMax=1.1, pitchMin=1.1}}
 {id=Modulator_5, children [Mixer_1]}
 {id=Random_1, children [WavePlayer_1,WavePlayer_6,WavePlayer_7,WavePlayer_8], properties {weights [1,1,1,1]}}
 {id=Random_3, children [WavePlayer_11, WavePlayer_12, WavePlayer_13], properties {weights [1,1,1]}}
 {id=WavePlayer_1, properties {soundWaveAssetPtr=&/CowCatcher_Deploy_01}}
 {id=WavePlayer_11, properties {soundWaveAssetPtr=&/CowCatcher_Deploy_LayerA_01}}
 {id=WavePlayer_12, properties {soundWaveAssetPtr=&/CowCatcher_Deploy_LayerA_02}}
 {id=WavePlayer_13, properties {soundWaveAssetPtr=&/CowCatcher_Deploy_LayerA_03}}
 {id=WavePlayer_17, properties {soundWaveAssetPtr=&/CowCatcher_Deploy_LayerG_01}}
 {id=WavePlayer_6, properties {soundWaveAssetPtr=&/CowCatcher_Deploy_02}}
 {id=WavePlayer_7, properties {soundWaveAssetPtr=&/CowCatcher_Deploy_03}}
 {id=WavePlayer_8, properties {soundWaveAssetPtr=&/CowCatcher_Deploy_04}}
]
properties {
 attenuationSettings=&/CowCatcher_Deploy_Att
 bOverrideConcurrency=true
 maxDistance=12600
 soundClassObject=&Fort_Vehicles
 volumeMultiplier=0.85
 attenuationOverrides {
  bAttenuateWithLPF=true
  distanceAlgorithm=Logarithmic
  falloffDistance=12000
  hPFFrequencyAtMax=100
  lPFFrequencyAtMax=3000
  lPFRadiusMax=2000
  lPFRadiusMin=0
  nonSpatializedRadiusStart=600
  stereoSpread=1000
  attenuationShapeExtents {x=600}
 }
 concurrencyOverrides {retriggerTime=0.3}
}
