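# Sound cue for the CowCatcher destruction sound in the CowCatcher mod, routed to the Fort_Vehicles sound class.
# A Random node picks among three equally weighted destroy waves, then a Modulator slightly varies the volume on each play.
# Playback uses a reduced cue volume and the shared gadget-placement attenuation settings.
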SoundCue /CowCatcherMod/Sound/CowCatcher_Destroy_Cue
root=Modulator_0
nodes [
 {id=Modulator_0, children [Random_1], properties {volumeMax=0.95, volumeMin=0.85}}
 {id=Random_1, children [WavePlayer_0,WavePlayer_4,WavePlayer_5], properties {weights [1,1,1]}}
 {id=WavePlayer_0, properties {soundWaveAssetPtr=&/CowCatcher_Destroy_01}}
 {id=WavePlayer_4, properties {soundWaveAssetPtr=&/CowCatcher_Destroy_02}}
 {id=WavePlayer_5, properties {soundWaveAssetPtr=&/CowCatcher_Destroy_03}}
]
properties {
 attenuationSettings=/Game/Sounds/Attenuation/Fort_GadgetPlace_Att
 maxDistance=2000
 soundClassObject=&Fort_Vehicles
 volumeMultiplier=0.65
}
