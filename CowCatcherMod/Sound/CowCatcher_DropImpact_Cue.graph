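# CowCatcher drop impact sound cue, providing the sound of the CowCatcher vehicle mod hitting on a drop.
# A Random node picks one of three equally weighted drop-impact waves and feeds it through a Modulator for variation.
# It plays at reduced volume in the vehicles sound class, spatialized with the borrowed gadget-throw attenuation preset.
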SoundCue /CowCatcherMod/Sound/CowCatcher_DropImpact_Cue
root=Modulator_0
nodes [
 {id=Modulator_0, children [Random_0]}
 {id=Random_0, children [WavePlayer_0,WavePlayer_2,WavePlayer_3], properties {weights [1,1,1]}}
 {id=WavePlayer_0, properties {soundWaveAssetPtr=&/CowCatcher_DropImpact_01}}
 {id=WavePlayer_2, properties {soundWaveAssetPtr=&/CowCatcher_DropImpact_02}}
 {id=WavePlayer_3, properties {soundWaveAssetPtr=&/CowCatcher_DropImpact_03}}
]
properties {
 attenuationSettings=/Game/Sounds/Attenuation/Fort_GadgetThrow_Att
 maxDistance=1536
 soundClassObject=&Fort_Vehicles
 volumeMultiplier=0.3
}
