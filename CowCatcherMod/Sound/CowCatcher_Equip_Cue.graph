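# Sound cue for equipping the CowCatcher in the CowCatcherMod, classed as weapon foley.
# A Random node picks one of three equip waves with equal weight, then a Modulator varies it, with the maximum pitch raised slightly above normal.
# Playback uses reload-style weapon attenuation and a reduced volume multiplier.
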SoundCue /CowCatcherMod/Sound/CowCatcher_Equip_Cue
root=Modulator_0
nodes [
 {id=Modulator_0, children [Random_1], properties {pitchMax=1.1}}
 {id=Random_1, children [WavePlayer_7,WavePlayer_8,WavePlayer_9], properties {weights [1,1,1]}}
 {id=WavePlayer_7, properties {soundWaveAssetPtr=&/CowCatcher_Equip_01}}
 {id=WavePlayer_8, properties {soundWaveAssetPtr=&/CowCatcher_Equip_02}}
 {id=WavePlayer_9, properties {soundWaveAssetPtr=&/CowCatcher_Equip_03}}
]
properties {
 attenuationSettings=&Fort_Weapon_Reload_Att
 maxDistance=2250
 soundClassObject=&Fort_WeaponFoley
 volumeMultiplier=0.401
}
