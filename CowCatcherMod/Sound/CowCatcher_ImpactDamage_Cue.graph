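# Sound cue for CowCatcher impact damage in the CowCatcher mod, layering two randomized hit sounds after a brief fixed delay.
# A mixer blends a random pick from borrowed security turret impact waves with a slightly quieter random CowCatcher impact wave.
# Each layer has its own modulator for variation, with the CowCatcher layer pitched upward.
# It uses small-object demolition hit attenuation and a concurrency override that stops the oldest instance.
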SoundCue /CowCatcherMod/Sound/CowCatcher_ImpactDamage_Cue
root=Delay_1
nodes [
 {id=Delay_1, children [Mixer_0], properties {delayMax=0.05, delayMin=0.05}}
 {id=Mixer_0, children [Modulator_1,Modulator_0], properties {inputVolume [1,0.9]}}
 {id=Modulator_0, children [Random_0], properties {pitchMax=1.3, pitchMin=1}}
 {id=Modulator_1, children [Random_3], properties {pitchMax=1}}
 {
  id=Random_0
  children [WavePlayer_0,WavePlayer_1,WavePlayer_2,WavePlayer_3,WavePlayer_4,WavePlayer_5,WavePlayer_6,WavePlayer_7]
  properties {weights [1,1,1,1,1,1,1,1]}
 }
 {id=Random_3, children [WavePlayer_13, WavePlayer_14, WavePlayer_16, WavePlayer_17], properties {weights [1,1,1,1]}}
 {id=WavePlayer_0, properties {soundWaveAssetPtr=&/CowCatcher_ImpactDamage_01}}
 {id=WavePlayer_1, properties {soundWaveAssetPtr=&/CowCatcher_ImpactDamage_02}}
 {id=WavePlayer_13, properties {soundWaveAssetPtr=/Game/Athena/Sounds/Gadgets/SecurityTurret/Security_Turret_Impact_02}}
 {id=WavePlayer_14, properties {soundWaveAssetPtr=/Game/Athena/Sounds/Gadgets/SecurityTurret/Security_Turret_Impact_01}}
 {id=WavePlayer_16, properties {soundWaveAssetPtr=/Game/Athena/Sounds/Gadgets/SecurityTurret/Security_Turret_Impact_04}}
 {id=WavePlayer_17, properties {soundWaveAssetPtr=/Game/Athena/Sounds/Gadgets/SecurityTurret/Security_Turret_Impact_05}}
 {id=WavePlayer_2, properties {soundWaveAssetPtr=&/CowCatcher_ImpactDamage_03}}
 {id=WavePlayer_3, properties {soundWaveAssetPtr=&/CowCatcher_ImpactDamage_04}}
 {id=WavePlayer_4, properties {soundWaveAssetPtr=&/CowCatcher_ImpactDamage_05}}
 {id=WavePlayer_5, properties {soundWaveAssetPtr=&/CowCatcher_ImpactDamage_06}}
 {id=WavePlayer_6, properties {soundWaveAssetPtr=&/CowCatcher_ImpactDamage_07}}
 {id=WavePlayer_7, properties {soundWaveAssetPtr=&/CowCatcher_ImpactDamage_08}}
]
properties {
 attenuationSettings=/Game/Sounds/Attenuation/Fort_Demolition_Hit_Small_Sized_Object_Att
 bOverrideConcurrency=true
 maxDistance=4000
 soundClassObject=&Fort_Impacts
 volumeMultiplier=0.64
 concurrencyOverrides {maxCount=3, resolutionRule=StopOldest, volumeScale=0.9, volumeScaleAttackTime=0.1}
}
