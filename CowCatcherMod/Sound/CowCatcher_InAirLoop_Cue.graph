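# Sound cue for the CowCatcher projectile's in-air sound, classed as a ranged weapon projectile effect.
# A random node picks one of three equally weighted Throw waves, which passes through slight pitch modulation into a mixer whose other input is empty.
# Despite the InAirLoop name, the graph has no looping node.
# It references a dedicated attenuation asset and overrides concurrency with a small per-owner voice limit.
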SoundCue /CowCatcherMod/Sound/CowCatcher_InAirLoop_Cue
root=Mixer_0
nodes [
 {id=Mixer_0, children [null,Modulator_0], properties {inputVolume [1,1]}}
 {id=Modulator_0, children [Random_0], properties {pitchMax=1.1, pitchMin=0.9}}
 {id=Random_0, children [WavePlayer_1,WavePlayer_2,WavePlayer_3], properties {weights [1,1,1]}}
 {id=WavePlayer_1, properties {soundWaveAssetPtr=&/CowCatcher_Throw_01}}
 {id=WavePlayer_2, properties {soundWaveAssetPtr=&/CowCatcher_Throw_02}}
 {id=WavePlayer_3, properties {soundWaveAssetPtr=&/CowCatcher_Throw_03}}
]
properties {
 attenuationSettings=/Game/Sounds/Attenuation/Fort_CowCatcher_InAir_Att
 bOverrideConcurrency=true
 maxDistance=4800
 pitchMultiplier=1.1
 soundClassObject=&Fort_WeaponRanged_Projectile_Effects
 volumeMultiplier=0.6
 attenuationOverrides {distanceAlgorithm=Logarithmic, falloffDistance=2000, stereoSpread=512, attenuationShapeExtents {x=0}}
 concurrencyOverrides {bLimitToOwner=true, maxCount=3, volumeScale=0.7}
}
