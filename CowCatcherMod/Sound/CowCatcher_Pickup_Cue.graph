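# Sound cue for the CowCatcher pickup sound in the CowCatcher mod.
# A Random node picks one of three equally weighted pickup wave variations and feeds it through a Modulator at default settings for pitch and volume variation.
# The cue lowers overall volume and pitch, plays in the Fort_Vehicles sound class, and borrows the gadget-throw attenuation asset for distance falloff.
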SoundCue /CowCatcherMod/Sound/CowCatcher_Pickup_Cue
root=Modulator_0
nodes [
 {id=Modulator_0, children [Random_0]}
 {id=Random_0, children [WavePlayer_1,WavePlayer_2,WavePlayer_3], properties {weights [1,1,1]}}
 {id=WavePlayer_1, properties {soundWaveAssetPtr=&/CowCatcher_Pickup_01}}
 {id=WavePlayer_2, properties {soundWaveAssetPtr=&/CowCatcher_Pickup_02}}
 {id=WavePlayer_3, properties {soundWaveAssetPtr=&/CowCatcher_Pickup_03}}
]
properties {
 attenuationSettings=/Game/Sounds/Attenuation/Fort_GadgetThrow_Att
 maxDistance=1536
 pitchMultiplier=0.9
 soundClassObject=&Fort_Vehicles
 volumeMultiplier=0.3
}
