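# CowCatcher_Throw_Cue is a sound cue in CowCatcherMod, named for a throw sound, with no sound nodes or root, so it has nothing to play.
# It holds only cue-level settings: gadget-throw attenuation, the Fort_Vehicles sound class, a limited max distance, lowered volume and raised pitch.
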SoundCue /CowCatcherMod/Sound/CowCatcher_Throw_Cue
root=null
nodes []
properties {
 attenuationSettings=/Game/Sounds/Attenuation/Fort_GadgetThrow_Att
 maxDistance=2000
 pitchMultiplier=2
 soundClassObject=&Fort_Vehicles
 volumeMultiplier=0.2
}
