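# CowCatcher_Whip_Cue is the whip sound cue for the CowCatcher mod.
# A Random node picks one of three equally weighted whip waves, and a Modulator varies its pitch and volume slightly on each play.
# It uses a borrowed weapon projectile attenuation, a third-person ranged-weapon fire sound class, and overridden concurrency that blocks new instances once the limit is reached.
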SoundCue /CowCatcherMod/Sound/CowCatcher_Whip_Cue
root=Modulator_0
nodes [
 {id=Modulator_0, children [Random_0], properties {pitchMax=1.2, pitchMin=0.8, volumeMax=1.15, volumeMin=0.85}}
 {id=Random_0, children [WavePlayer_1,WavePlayer_2,WavePlayer_3], properties {weights [1,1,1]}}
 {id=WavePlayer_1, properties {soundWaveAssetPtr=&/CowCatcher_Whip_01}}
 {id=WavePlayer_2, properties {soundWaveAssetPtr=&/CowCatcher_Whip_02}}
 {id=WavePlayer_3, properties {soundWaveAssetPtr=&/CowCatcher_Whip_03}}
]
properties {
 attenuationSettings=/Game/Sounds/Attenuation/Fort_Weapon_Launcher_Proj_InAir_Default_Att
 bOverrideConcurrency=true
 maxDistance=12750
 soundClassObject=&Fort_WeaponRanged_Fire_3P_PSM
 volumeMultiplier=0.7
 attenuationOverrides {falloffDistance=5000, attenuationShapeExtents {x=1000}}
 concurrencyOverrides {resolutionRule=PreventNew}
}
